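# Translucent unlit glow material for Niagara sprites in the Sparks stage show, drawing a soft radial spot in stage colors.
# A dynamic ColorSelect parameter blends MF_MojoStageColors outputs from Primary to Tertiary, although the first blend uses Primary on both sides.
# The color is tinted by particle color and passed through MF_FX_EmissiveAdjustments.
# A UV SphereMask with Radius, Hardness and Power falloff controls, multiplied by particle alpha, shapes the glow.
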Material /SparksMaps/Effects/Materials/Glow/M_Glow_StageShow
graph {
 properties {bUsedWithNiagaraSprites=true, blendMode=BLEND_Translucent, pixelDepthOffsetMode=PDOM_Legacy, shadingModel=MSM_Unlit, usedShadingModels=Unlit}
 nodes [
  n1 = MaterialFunctionCall!(CustomUPosition) {materialFunction=/SparksStageShow/Systems/StageManager/FXManager/MaterialFunctions/MF_MojoStageColors}
  n2 = DynamicParameter() {defaultValue=<rgba 0 1 1 1>, paramNames [ColorSelect,Param2,Param3,Param4]}
  n3 = LinearInterpolate(A=@n1.Primary, B=@n1.Primary, Alpha=Saturate(None=@n2.ColorSelect))
  n4 = Saturate(None=Subtract(A=@n2.ColorSelect))
  n5 = ParticleColor()
  MaterialFunctionCall(Emissive=Multiply(A=LinearInterpolate(A=@n3, B=@n1.Tertiary, Alpha=@n4), B=@n5.RGB)) {materialFunction=&MF_FX_EmissiveAdjustments}
  n6 = TextureCoordinate()
  n7 = Constant() {r=0.5}
  n8 = ScalarParameter() {defaultValue=0.5, parameterName=Radius}
  n9 = ScalarParameter() {parameterName=Hardness}
  Multiply(
   A=@n5.A
   B=Power(Base=SphereMask(A=@n6, B=@n7, Radius=@n8, Hardness=@n9) {hardnessPercent=0}, Exp=ScalarParameter() {defaultValue=3, parameterName=Power}) {constExponent=3}
  )
 ]
}
