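# Additive unlit glow material for stage show effects, usable on Niagara sprites.
# An exponential radial gradient is multiplied by the primary color from MF_MojoStageColors to form the tinted glow.
# A separate "tentacles" branch builds two panning band layers in radial coordinates, wobbled by time-driven sine and panning noise, then masked by another gradient.
# No shown node reads that tentacles output.
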Material /SparksMaps/Effects/Materials/Glow/M_Glow_StageShow_2
graph {
 properties {bUsedWithNiagaraSprites=true, blendMode=BLEND_Additive, pixelDepthOffsetMode=PDOM_Legacy, shadingModel=MSM_Unlit, usedShadingModels=Unlit}
 nodes [
  n1 = Constant() {r=1}
  Multiply(
   A=MaterialFunctionCall!(CustomUPosition) {materialFunction=/SparksStageShow/Systems/StageManager/FXManager/MaterialFunctions/MF_MojoStageColors}.Primary
   B=Multiply(A=MaterialFunctionCall(Density=@n1) {materialFunction=&RadialGradientExponential}.RadialGradientExponential) {constB=2}
  )
  n2 = Divide(A=Subtract(A=MaterialFunctionCall() {materialFunction=&RadialGradientExponential}.RadialGradientExponential) {constB=0.01})
  n3 = Multiply(
   A=MaterialFunctionCall() {materialFunction=&VectorToRadialValue}.Radial Coordinates
   B=Constant2Vector() {g=1, r=24}
  )
  n4 = TextureCoordinate()
  n5 = Constant() {r=2}
  n6 = Add(A=ComponentMask(None=Rotator(Coordinate=@n4, Time=@n5)) {r=true}, B=Multiply(A=Time(), B=Constant() {r=0.15}))
  n7 = Constant() {r=2}
  n8 = TextureCoordinate()
  n9 = AppendVector(A=Constant() {r=0.5}, B=Constant() {r=0.5})
  n10 = Constant() {r=0.15}
  n11 = Constant() {r=0.1}
  n12 = TextureSample(UVs=Panner(Coordinate=Multiply(A=@n8, B=@n9), Speed=AppendVector(A=@n10, B=@n11)) {speedX=0.2, speedY=0.15}) {samplerType=SAMPLERTYPE_Grayscale, texture=&LowResBlurredNoise}
  n13 = Add(
   A=Sine(None=Multiply(A=@n6, B=@n7))
   B=Multiply(A=NamedRerouteDeclaration(None=@n12.R) {name=NoiseTexture, nodeColor=<rgba 0 0 0 1>}, B=Constant() {r=2})
  )
  n14 = Multiply(A=@n13, B=ScalarParameter() {defaultValue=0.5, parameterName=Rays 1 Wobble Amount}) {constB=0.001}
  n15 = MaterialFunctionCall!(
   Width=Constant() {r=0.1}
   Sharpness=Constant() {r=0.95}
   Offset
   Direction Switch=StaticBool() {value=true}
   Compare
   Input Coordinates=Panner(Coordinate=Add(A=@n3, B=@n14)) {speedX=-0.1}
  ) {materialFunction=/Engine/Functions/Engine_MaterialFunctions03/Procedurals/GeneratedBand}
  n16 = Add(
   A=Multiply(A=@n13, B=ScalarParameter() {defaultValue=0.5, parameterName=Rays 2 Wobble Amount})
   B=Multiply(
    A=MaterialFunctionCall() {materialFunction=&VectorToRadialValue}.Radial Coordinates
    B=Constant2Vector() {g=1, r=24}
   )
  )
  n17 = Multiply(A=MaterialFunctionCall(
    Width=Constant() {r=0.3}
    Sharpness=Constant() {r=0.95}
    Direction Switch=StaticBool() {value=true}
    Input Coordinates=Panner(Coordinate=@n16) {speedX=-0.1}
   ) {materialFunction=&GeneratedBand}.Result) {constB=0.1}
  NamedRerouteDeclaration(None=Multiply(A=Multiply(A=@n2), B=Add(A=@n15.Result, B=@n17))) {name=tentacles, nodeColor=<rgba 0.5469 1 0 1>}
 ]
}
