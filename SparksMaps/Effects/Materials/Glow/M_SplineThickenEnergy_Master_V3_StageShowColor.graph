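# Translucent, unlit master effect material for energy strands thickened by SplineThicken, with optional stage show coloring.
# SplineThicken drives world position offset with base-to-tip widths and supplies normals for an inverted fresnel edge fade.
# Opacity combines that fade, a texture panned by dynamic parameter, side, base and tip masks, alpha and optional depth fade.
# Emissive switches between color/particle color and MF_MojoStageColors, whose Primary output feeds all three Lerp_3Color color inputs.
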Material /SparksMaps/Effects/Materials/Glow/M_SplineThickenEnergy_Master_V3_StageShowColor
graph {
 properties {
  bUsedWithMeshDeformer=true
  bUsedWithNanite=true
  bUsedWithNiagaraMeshParticles=true
  bUsedWithSkeletalMesh=true
  blendMode=BLEND_Translucent
  pixelDepthOffsetMode=PDOM_Legacy
  shadingModel=MSM_Unlit
  twoSided=true
  usedShadingModels=Unlit
 }
 nodes [
  n1 = MaterialFunctionCall!(CustomUPosition) {materialFunction=/SparksStageShow/Systems/StageManager/FXManager/MaterialFunctions/MF_MojoStageColors}
  n2 = Saturate(None=ScalarParameter() {parameterName=StageShowColorSelect})
  n3 = VectorParameter() {parameterName=Color, defaultValue=<rgba 1 1 1 0>}
  n4 = ParticleColor()
  n5 = ScalarParameter() {defaultValue=1, parameterName=Use Particle Color}
  MaterialFunctionCall(Emissive=StaticSwitchParameter(
    True=MaterialFunctionCall(A=@n1.Primary, B=@n1.Primary, C=@n1.Primary, Alpha=@n2) {materialFunction=&Lerp_3Color}.Result
    False=LinearInterpolate(A=@n3.RGB, B=@n4.RGB, Alpha=@n5)
   ) {parameterName=UseStageShowColors}) {materialFunction=&MF_FX_EmissiveAdjustments}
  n6 = Constant() {r=1}
  n7 = ScalarParameter() {defaultValue=1.4, group=UV Settings, parameterName=V Scale}
  n8 = Multiply(A=Multiply(A=TextureCoordinate(), B=AppendVector(A=@n6, B=@n7)), B=Constant2Vector() {g=-1, r=1})
  n9 = DynamicParameter() {paramNames [Pan Speed, Param2, Param3, Param4]}
  n10 = Constant()
  n11 = TextureSampleParameter2D(UVs=Add(
    A=MaterialFunctionCall(XY=@n8) {materialFunction=&Swizzle}.YX
    B=AppendVector(A=@n9.Pan Speed, B=@n10)
   )) {group=Textures, parameterName=Texture Alpha, texture=/Game/Effects/Fort_Effects/Gameplay/SmokeDustExplosions/Textures/T_RibbonSmokeMasked_05}
  n12 = Constant() {r=0.5}
  n13 = TextureCoordinate()
  n14 = SphereMask(
   A=AppendVector(
    A=@n12
    B=MaterialFunctionCall(Float2=@n13) {materialFunction=&BreakOutFloat2Components}.R
   )
   B=@n12
   Radius=ScalarParameter() {defaultValue=0.5, group=Gradient Mask Settings, parameterName=Side Mask Radius, sortPriority=4}
   Hardness=ScalarParameter() {defaultValue=0.5, group=Gradient Mask Settings, parameterName=Side Mask Hardness, sortPriority=4}
  )
  n15 = MaterialFunctionCall(Float2=TextureCoordinate()) {materialFunction=&BreakOutFloat2Components}
  n16 = Power(
   Base=OneMinus(None=@n15.G)
   Exp=ScalarParameter() {defaultValue=2, group=Gradient Mask Settings, parameterName=Base Mask Softness, sortPriority=4}
  )
  n17 = ScalarParameter() {defaultValue=12, group=Gradient Mask Settings, parameterName=Base Mask Multiplier, sortPriority=4}
  n18 = Power(Base=@n15.G, Exp=ScalarParameter() {defaultValue=0.2, group=Gradient Mask Settings, parameterName=Tip Mask Softness, sortPriority=4})
  n19 = ScalarParameter() {defaultValue=2, group=Gradient Mask Settings, parameterName=Tip Mask Multiplier, sortPriority=4}
  n20 = Multiply(A=Saturate(None=Multiply(A=@n16, B=@n17)), B=Saturate(None=Multiply(A=@n18, B=@n19)))
  n21 = ScalarParameter() {group=Fresnel Settings, parameterName=FresnelExponentIn, sortPriority=1}
  n22 = MaterialFunctionCall(
   WidthBase=ScalarParameter() {defaultValue=24, group=Spline Settings, parameterName=Spline Width Base, sortPriority=1}
   WidthTip=ScalarParameter() {defaultValue=6, group=Spline Settings, parameterName=Spline Width Tip, sortPriority=1}
   AngleCorrectedNormals=StaticBool() {value=true}
  ) {materialFunction=&SplineThicken}
  n23 = OneMinus(None=Saturate(None=Fresnel(ExponentIn=@n21, Normal=@n22.Normal) {exponent=1}))
  n24 = ScalarParameter() {defaultValue=1, group=Fresnel Settings, parameterName=FresnelExponent, sortPriority=1}
  n25 = Multiply(A=@n11.RGB, B=Multiply(A=Multiply(A=@n14, B=@n20), B=Power(Base=@n23, Exp=@n24) {constExponent=6}))
  n26 = DepthFade(FadeDistance=ScalarParameter() {defaultValue=50, group=Depth Fade, parameterName=Depth Fade})
  Multiply(
   A=LinearInterpolate(A=@n3.A, B=@n4.A, Alpha=@n5)
   B=StaticSwitchParameter(True=Multiply(A=@n25, B=@n26), False=@n25) {group=Depth Fade, parameterName=Enable Depth Fade}
  )
  Reroute(None=@n22.WorldPositionOffset)
 ]
}
